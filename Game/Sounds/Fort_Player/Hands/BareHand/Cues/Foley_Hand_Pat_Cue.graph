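# Sound cue for player bare-hand pat foley.
# A Random node picks one of six pat wave variations with equal weight, and a Modulator at the root varies pitch slightly while leaving volume fixed.
# It plays through a first-person player foley sound class but uses the third-person clothing attenuation asset.
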SoundCue /Game/Sounds/Fort_Player/Hands/BareHand/Cues/Foley_Hand_Pat_Cue
root=Modulator_0
nodes [
 {id=Modulator_0, children [Random_0], properties {pitchMax=1.2, pitchMin=0.9, volumeMax=1, volumeMin=1}}
 {id=Random_0, children [WavePlayer_0,WavePlayer_1,WavePlayer_2,WavePlayer_3,WavePlayer_4,WavePlayer_5], properties {weights [1,1,1,1,1,1]}}
 {id=WavePlayer_0, properties {soundWaveAssetPtr=/Game/Sounds/Fort_Player/Hands/BareHand/Waves/Pat/Foley_Hands_Pat_01}}
 {id=WavePlayer_1, properties {soundWaveAssetPtr=/Game/Sounds/Fort_Player/Hands/BareHand/Waves/Pat/Foley_Hands_Pat_02}}
 {id=WavePlayer_2, properties {soundWaveAssetPtr=/Game/Sounds/Fort_Player/Hands/BareHand/Waves/Pat/Foley_Hands_Pat_03}}
 {id=WavePlayer_3, properties {soundWaveAssetPtr=/Game/Sounds/Fort_Player/Hands/BareHand/Waves/Pat/Foley_Hands_Pat_04}}
 {id=WavePlayer_4, properties {soundWaveAssetPtr=/Game/Sounds/Fort_Player/Hands/BareHand/Waves/Pat/Foley_Hands_Pat_05}}
 {id=WavePlayer_5, properties {soundWaveAssetPtr=/Game/Sounds/Fort_Player/Hands/BareHand/Waves/Pat/Foley_Hands_Pat_06}}
]
properties {
 attenuationSettings=/Game/Sounds/Fort_Player/Clothing/Settings/Fort_Player_Clothing_3P_Att
 maxDistance=2000
 soundClassObject=&Fort_Player_Foley_1P_Athena
 volumeMultiplier=1
}
