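# Sound cue for player bare-hand foley of patting fur.
# A random node picks one of six equally weighted pat recordings, then a modulator varies its pitch and volume slightly on each play.
# The cue boosts overall volume and uses the first-person Athena player foley sound class, paired with third-person clothing attenuation settings.
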SoundCue /Game/Sounds/Fort_Player/Hands/BareHand/Cues/Foley_Hand_Pat_Fur_Cue
root=Modulator_0
nodes [
 {id=Modulator_0, children [Random_1], properties {pitchMax=1.1, pitchMin=0.8, volumeMax=1.1, volumeMin=0.9}}
 {
  id=Random_1
  children [WavePlayer_6, WavePlayer_7, WavePlayer_8, WavePlayer_9, WavePlayer_10, WavePlayer_11]
  properties {weights [1,1,1,1,1,1]}
 }
 {id=WavePlayer_10, properties {soundWaveAssetPtr=/Game/Sounds/Fort_Player/Hands/BareHand/Waves/Pat/Foley_Hands_Pat_Fur_05}}
 {id=WavePlayer_11, properties {soundWaveAssetPtr=/Game/Sounds/Fort_Player/Hands/BareHand/Waves/Pat/Foley_Hands_Pat_Fur_06}}
 {id=WavePlayer_6, properties {soundWaveAssetPtr=/Game/Sounds/Fort_Player/Hands/BareHand/Waves/Pat/Foley_Hands_Pat_Fur_01}}
 {id=WavePlayer_7, properties {soundWaveAssetPtr=/Game/Sounds/Fort_Player/Hands/BareHand/Waves/Pat/Foley_Hands_Pat_Fur_02}}
 {id=WavePlayer_8, properties {soundWaveAssetPtr=/Game/Sounds/Fort_Player/Hands/BareHand/Waves/Pat/Foley_Hands_Pat_Fur_03}}
 {id=WavePlayer_9, properties {soundWaveAssetPtr=/Game/Sounds/Fort_Player/Hands/BareHand/Waves/Pat/Foley_Hands_Pat_Fur_04}}
]
properties {
 attenuationSettings=/Game/Sounds/Fort_Player/Clothing/Settings/Fort_Player_Clothing_3P_Att
 maxDistance=2000
 soundClassObject=&Fort_Player_Foley_1P_Athena
 volumeMultiplier=3
}
