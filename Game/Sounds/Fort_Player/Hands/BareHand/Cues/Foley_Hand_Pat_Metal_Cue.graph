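# Sound cue for bare-hand pat foley on metal for the player character.
# A Random node picks one of four equally weighted pat wave variations.
# The cue applies boosted volume and slightly lowered pitch.
# It routes to the first-person Athena player foley sound class but uses the third-person clothing attenuation asset.
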SoundCue /Game/Sounds/Fort_Player/Hands/BareHand/Cues/Foley_Hand_Pat_Metal_Cue
root=Random_1
nodes [
 {id=Random_1, children [WavePlayer_6,WavePlayer_7,WavePlayer_8,WavePlayer_9], properties {weights [1,1,1,1]}}
 {id=WavePlayer_6, properties {soundWaveAssetPtr=/Game/Sounds/Fort_Player/Hands/BareHand/Waves/Pat/Foley_Hands_Pat_Metal_01}}
 {id=WavePlayer_7, properties {soundWaveAssetPtr=/Game/Sounds/Fort_Player/Hands/BareHand/Waves/Pat/Foley_Hands_Pat_Metal_02}}
 {id=WavePlayer_8, properties {soundWaveAssetPtr=/Game/Sounds/Fort_Player/Hands/BareHand/Waves/Pat/Foley_Hands_Pat_Metal_03}}
 {id=WavePlayer_9, properties {soundWaveAssetPtr=/Game/Sounds/Fort_Player/Hands/BareHand/Waves/Pat/Foley_Hands_Pat_Metal_04}}
]
properties {
 attenuationSettings=/Game/Sounds/Fort_Player/Clothing/Settings/Fort_Player_Clothing_3P_Att
 maxDistance=2000
 pitchMultiplier=0.9
 soundClassObject=&Fort_Player_Foley_1P_Athena
 volumeMultiplier=2
}
