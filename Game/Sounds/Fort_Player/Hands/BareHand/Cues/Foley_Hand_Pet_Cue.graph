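# Sound cue for bare-hand petting foley on the player, routed to the Athena player foley sound class with third-person clothing attenuation.
# A quality-level switch feeds the first two levels from an eleven-wave random pool and the last two from a four-wave subset to trim variation.
# A modulator randomizes volume while leaving pitch fixed, and the cue itself plays the rub recordings quieter and at raised pitch.
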SoundCue /Game/Sounds/Fort_Player/Hands/BareHand/Cues/Foley_Hand_Pet_Cue
root=Modulator_0
nodes [
 {id=Modulator_0, children [QualityLevel_0], properties {pitchMax=1, pitchMin=1, volumeMax=1, volumeMin=0.3}}
 {id=QualityLevel_0, children [Random_0,Random_0,Random_3,Random_3]}
 {
  id=Random_0
  children [
   WavePlayer_1
   WavePlayer_2
   WavePlayer_3
   WavePlayer_4
   WavePlayer_5
   WavePlayer_6
   WavePlayer_7
   WavePlayer_8
   WavePlayer_9
   WavePlayer_10
   WavePlayer_11
  ]
  properties {weights [1,1,1,1,1,1,1,1,1,1,1]}
 }
 {id=Random_3, children [WavePlayer_12, WavePlayer_13, WavePlayer_14, WavePlayer_15], properties {weights [1,1,1,1]}}
 {id=WavePlayer_1, properties {soundWaveAssetPtr=/Game/Sounds/Fort_Player/Hands/BareHand/Waves/Rub/Foley_Hands_Pet_01}}
 {id=WavePlayer_10, properties {soundWaveAssetPtr=/Game/Sounds/Fort_Player/Hands/BareHand/Waves/Rub/Foley_Hands_Pet_10}}
 {id=WavePlayer_11, properties {soundWaveAssetPtr=/Game/Sounds/Fort_Player/Hands/BareHand/Waves/Rub/Foley_Hands_Pet_11}}
 {id=WavePlayer_12, properties {soundWaveAssetPtr=&Foley_Hands_Pet_01}}
 {id=WavePlayer_13, properties {soundWaveAssetPtr=/Game/Sounds/Fort_Player/Hands/BareHand/Waves/Rub/Foley_Hands_Pet_02}}
 {id=WavePlayer_14, properties {soundWaveAssetPtr=/Game/Sounds/Fort_Player/Hands/BareHand/Waves/Rub/Foley_Hands_Pet_03}}
 {id=WavePlayer_15, properties {soundWaveAssetPtr=/Game/Sounds/Fort_Player/Hands/BareHand/Waves/Rub/Foley_Hands_Pet_04}}
 {id=WavePlayer_2, properties {soundWaveAssetPtr=&Foley_Hands_Pet_02}}
 {id=WavePlayer_3, properties {soundWaveAssetPtr=&Foley_Hands_Pet_03}}
 {id=WavePlayer_4, properties {soundWaveAssetPtr=&Foley_Hands_Pet_04}}
 {id=WavePlayer_5, properties {soundWaveAssetPtr=/Game/Sounds/Fort_Player/Hands/BareHand/Waves/Rub/Foley_Hands_Pet_05}}
 {id=WavePlayer_6, properties {soundWaveAssetPtr=/Game/Sounds/Fort_Player/Hands/BareHand/Waves/Rub/Foley_Hands_Pet_06}}
 {id=WavePlayer_7, properties {soundWaveAssetPtr=/Game/Sounds/Fort_Player/Hands/BareHand/Waves/Rub/Foley_Hands_Pet_07}}
 {id=WavePlayer_8, properties {soundWaveAssetPtr=/Game/Sounds/Fort_Player/Hands/BareHand/Waves/Rub/Foley_Hands_Pet_08}}
 {id=WavePlayer_9, properties {soundWaveAssetPtr=/Game/Sounds/Fort_Player/Hands/BareHand/Waves/Rub/Foley_Hands_Pet_09}}
]
properties {
 attenuationSettings=/Game/Sounds/Fort_Player/Clothing/Settings/Fort_Player_Clothing_3P_Att
 maxDistance=2000
 pitchMultiplier=2
 soundClassObject=&Fort_Player_Foley_Athena
 volumeMultiplier=0.4
}
